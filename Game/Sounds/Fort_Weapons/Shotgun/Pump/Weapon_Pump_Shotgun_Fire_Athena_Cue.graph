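# Sound cue for the Athena pump shotgun's gunfire, using third-person and distant ranged-fire sound classes.
# A mixer layers close, medium and distant shots at descending input volumes, each with its own sound class override and attenuation asset.
# Each layer randomly picks one of two waves through a modulator, with volume boosts on the medium and distant layers.
# Local and global gunfire concurrency sets are assigned.
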SoundCue /Game/Sounds/Fort_Weapons/Shotgun/Pump/Weapon_Pump_Shotgun_Fire_Athena_Cue
root=Mixer_0
nodes [
 {id=Attenuation_0, children [Modulator_0], properties {attenuationSettings=&Fort_Weapon_Projectile_Close_Athena_Att}}
 {
  id=Attenuation_1
  children [Modulator_1]
  properties {attenuationSettings=&Fort_Weapon_Projectile_Medium_Mono_Athena_Att}
 }
 {
  id=Attenuation_2
  children [Modulator_2]
  properties {attenuationSettings=&Fort_Weapon_Projectile_Distant_Mono_Athena_Att}
 }
 {id=Mixer_0, children [SoundClass_0,SoundClass_6,SoundClass_1], properties {inputVolume [1,0.2,0.1]}}
 {id=Modulator_0, children [Random_3], properties {pitchMin=1}}
 {id=Modulator_1, children [Random_4], properties {volumeMax=2, volumeMin=2}}
 {id=Modulator_2, children [Random_5], properties {volumeMax=1.5, volumeMin=1.5}}
 {id=Random_3, children [WavePlayer_13, WavePlayer_17], properties {bShouldExcludeFromBranchCulling=true, weights [1,1]}}
 {id=Random_4, children [WavePlayer_18, WavePlayer_19], properties {bShouldExcludeFromBranchCulling=true, weights [1,1]}}
 {id=Random_5, children [WavePlayer_20, WavePlayer_21], properties {bShouldExcludeFromBranchCulling=true, weights [1,1]}}
 {
  id=SoundClass_0
  children [Attenuation_0]
  properties {soundClassOverride=&Fort_WeaponRanged_Fire_3P_PSM}
 }
 {
  id=SoundClass_1
  children [Attenuation_2]
  properties {soundClassOverride=&Fort_WeaponRanged_Fire_Distant}
 }
 {
  id=SoundClass_6
  children [Attenuation_1]
  properties {soundClassOverride=&Fort_WeaponRanged_Fire_3P_PSM}
 }
 {id=WavePlayer_13, properties {soundWaveAssetPtr=&/Shotgun_Pump_NewFire2_01}}
 {id=WavePlayer_17, properties {soundWaveAssetPtr=&/Shotgun_Pump_NewFire2_02}}
 {id=WavePlayer_18, properties {soundWaveAssetPtr=&/Shotgun_Pump_MediumFireNew_02}}
 {id=WavePlayer_19, properties {soundWaveAssetPtr=&/Shotgun_Pump_MediumFireNew_01}}
 {id=WavePlayer_20, properties {soundWaveAssetPtr=&/Shotgun_Pump_DistantFireNew_01}}
 {id=WavePlayer_21, properties {soundWaveAssetPtr=&/Shotgun_Pump_DistantFireNew_02}}
]
properties {
 maxDistance=27500
 soundClassObject=&Fort_WeaponRanged_Fire_3P_PSM
 volumeMultiplier=0.9
 attenuationOverrides {falloffDistance=5000, attenuationShapeExtents {x=1000}}
 concurrencyOverrides {bLimitToOwner=true, maxCount=3, resolutionRule=StopOldest}
 concurrencySet [&Weapon_Gunfire_Local_Conc, &Weapon_Gunfire_Global_Conc]
}
